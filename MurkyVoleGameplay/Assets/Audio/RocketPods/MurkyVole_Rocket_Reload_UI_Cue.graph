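# Sound cue for the MurkyVole rocket pods' reload UI sound.
# A random node picks one of three equally weighted "Rocket Ready" waves and feeds a modulator that pins pitch at normal.
# It plays in the weapon foley sound class with weapon reload attenuation, reduced volume and a concurrency override.
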SoundCue /MurkyVoleGameplay/Assets/Audio/RocketPods/MurkyVole_Rocket_Reload_UI_Cue
root=Modulator_0
nodes [
 {id=Modulator_0, children [Random_0], properties {pitchMax=1, pitchMin=1}}
 {id=Random_0, children [WavePlayer_2,WavePlayer_3,WavePlayer_4], properties {weights [1,1,1]}}
 {id=WavePlayer_2, properties {soundWaveAssetPtr=&/MurkyVole_Rocket_Ready_01}}
 {id=WavePlayer_3, properties {soundWaveAssetPtr=&/MurkyVole_Rocket_Ready_02}}
 {id=WavePlayer_4, properties {soundWaveAssetPtr=&/MurkyVole_Rocket_Ready_03}}
]
properties {
 attenuationSettings=&Fort_Weapon_Reload_Att
 bOverrideConcurrency=true
 maxDistance=2250
 soundClassObject=&Fort_WeaponFoley
 volumeMultiplier=0.6
}
