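# Sound cue for the MurkyVole rocket pods' "pods down" movement, classed as a gadgets and items sound.
# It picks one of three equally weighted wave variants at random, then passes it through a modulator that fixes pitch slightly below normal.
# It borrows the auto turret's ambient attenuation asset and carries listener-focus override values.
# A per-owner concurrency limit allows one instance at a time and blocks new plays while one is active.
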SoundCue /MurkyVoleGameplay/Assets/Audio/RocketPods/MurkyVole_Rockets_PodsDown_Cue
root=Modulator_1
nodes [
 {id=Modulator_1, children [Random_0], properties {pitchMax=0.9, pitchMin=0.9}}
 {id=Random_0, children [WavePlayer_1,WavePlayer_2,WavePlayer_3], properties {weights [1,1,1]}}
 {id=WavePlayer_1, properties {soundWaveAssetPtr=&/MurkyVole_Rockets_PodsDown_01}}
 {id=WavePlayer_2, properties {soundWaveAssetPtr=&/MurkyVole_Rockets_PodsDown_02}}
 {id=WavePlayer_3, properties {soundWaveAssetPtr=&/MurkyVole_Rockets_PodsDown_03}}
]
properties {
 attenuationSettings=/MurkyVoleGameplay/Assets/Audio/AutoTurret/Settings/MurkyVole_Turret_Ambient_Att
 bOverrideConcurrency=true
 maxDistance=1050
 soundClassObject=&Fort_Gadgets_Items
 volumeMultiplier=0.7
 attenuationOverrides {
  bEnableFocusInterpolation=true
  bEnableListenerFocus=true
  falloffDistance=1200
  nonFocusVolumeAttenuation=0.5
  stereoSpread=500
  attenuationShapeExtents {x=0}
 }
 concurrencyOverrides {bLimitToOwner=true, maxCount=1, resolutionRule=PreventNew}
}
